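# Looping Sound Cue that plays periodic PA announcements for the Utopia Academy.
# Each loop waits a random one to two minutes, then concatenates a chime with one of eight equally weighted random voice-over lines.
# Base submix and submix sends are disabled, so output goes only to the Academy PA audio bus at full send level.
# Play-when-silent virtualization keeps the loop running while inaudible.
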SoundCue /Hermes/Audio/Utopia/SC_Utopia_Academy_PA_Loop
root=Looping_0
nodes [
 {id=Concatenator_0, children [Mixer_0,Mixer_1], properties {inputVolume [1,1]}}
 {id=Delay_0, children [Concatenator_0], properties {delayMax=120, delayMin=60}}
 {id=Looping_0, children [Delay_0]}
 {id=Mixer_0, children [WavePlayer_1], properties {inputVolume [1]}}
 {id=Mixer_1, children [Random_0], properties {inputVolume [1]}}
 {
  id=Random_0
  children [WavePlayer_2,WavePlayer_3,WavePlayer_4,WavePlayer_5,WavePlayer_6,WavePlayer_7,WavePlayer_8,WavePlayer_9]
  properties {bShouldExcludeFromBranchCulling=true, weights [1,1,1,1,1,1,1,1]}
 }
 {id=WavePlayer_1, properties {soundWaveAssetPtr=&/Spot_PA_Utopia_Academy_Announcement_Chime_A}}
 {id=WavePlayer_2, properties {soundWaveAssetPtr=&/Spot_PA_Utopia_Academy_Announcement_VO_F_01}}
 {id=WavePlayer_3, properties {soundWaveAssetPtr=&/Spot_PA_Utopia_Academy_Announcement_VO_F_02}}
 {id=WavePlayer_4, properties {soundWaveAssetPtr=&/Spot_PA_Utopia_Academy_Announcement_VO_F_03}}
 {id=WavePlayer_5, properties {soundWaveAssetPtr=&/Spot_PA_Utopia_Academy_Announcement_VO_F_04}}
 {id=WavePlayer_6, properties {soundWaveAssetPtr=&/Spot_PA_Utopia_Academy_Announcement_VO_F_05}}
 {id=WavePlayer_7, properties {soundWaveAssetPtr=&/Spot_PA_Utopia_Academy_Announcement_VO_F_06}}
 {id=WavePlayer_8, properties {soundWaveAssetPtr=&/Spot_PA_Utopia_Academy_Announcement_VO_F_07}}
 {id=WavePlayer_9, properties {soundWaveAssetPtr=&/Spot_PA_Utopia_Academy_Announcement_VO_F_08}}
]
properties {
 bEnableBaseSubmix=false
 bEnableSubmixSends=false
 maxDistance=2097152
 virtualizationMode=PlayWhenSilent
 volumeMultiplier=1
 busSends [{audioBus=&/AB_MusicParty_Utopia_Academy_PA, minSendLevel=0, sendLevel=1, soundSourceBus=null}]
}
