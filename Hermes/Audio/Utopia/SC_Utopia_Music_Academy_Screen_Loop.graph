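# Sound cue for a looping music academy screen sound in the Utopia audio set, classed as an Athena ambient prop.
# The graph has no root or nodes, so it references no sound wave and has nothing to play as shown.
# Only cue-level settings remain: a dedicated attenuation asset with limited range, reduced volume and submix sends disabled.
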SoundCue /Hermes/Audio/Utopia/SC_Utopia_Music_Academy_Screen_Loop
root=null
nodes []
properties {
 attenuationSettings=&/ATT_Utopia_Music_Academy_Sceen
 bEnableSubmixSends=false
 maxDistance=2100
 soundClassObject=&Fort_Ambient_Props_Athena
 volumeMultiplier=0.3
}
