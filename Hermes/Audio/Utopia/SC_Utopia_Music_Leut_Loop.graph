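# Sound cue for the Utopia "Leut" music loop, playing the Shopping 01 music wave on repeat.
# A single looping wave player is the root, spatialized through a Utopia music attenuation asset with a 3000-unit max distance under the ambient props sound class.
# Base submix output and submix sends are disabled, so audio feeds only a pre-effect send into a dedicated source bus for this loop.
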SoundCue /Hermes/Audio/Utopia/SC_Utopia_Music_Leut_Loop
root=WavePlayer_0
nodes [{id=WavePlayer_0, properties {bLooping=true, soundWaveAssetPtr=&/MX_Utopia_Music_Shopping_01_Loop}}]
properties {
 attenuationSettings=&/ATT_Utopia_Music_3K
 bEnableBaseSubmix=false
 bEnableSubmixSends=false
 maxDistance=3000
 soundClassObject=&Fort_Ambient_Props_Athena
 volumeMultiplier=1
 preEffectBusSends [{soundSourceBus=&/SB_Utopia_Music_Leut_Loop}]
}
